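# Masked subsurface material for Niagara water bubble mesh particles, rendering round bubbles with a sphere normal map and fake reflection emissive.
# Vertex colors and dynamic parameters spread pieces along normals, size them through a sprite function, and wiggle them with sine/cosine, shrinking near the camera.
# Optional water displacement joins the WPO, which a custom node marks precise to fix depth variance.
# Biome tint, optional Cascade coloring and particle-driven HSV offsets adjust color.
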Material /Game/Effects/Niagara/Emitters/Water/Bubbles/M_NFX_Bubbles
graph {
 properties {
  bTangentSpaceNormal=false
  bUsedWithMeshParticles=true
  bUsedWithNanite=true
  bUsedWithNiagaraMeshParticles=true
  blendMode=BLEND_Masked
  pixelDepthOffsetMode=PDOM_Legacy
  shadingModel=MSM_Subsurface
  usedShadingModels=Subsurface
 }
 nodes [
  Constant()
  Constant() {r=0.5}
  n1 = ComponentMask(None=VertexNormalWS()) {g=true, r=true}
  n2 = Constant()
  n3 = DynamicParameter() {parameterIndex=1, defaultValue=<rgba 0 0 0 0>, paramNames [RemoveSeperateZ, Param2, Param3, Param4]}
  n4 = VertexColor()
  n5 = DynamicParameter() {defaultValue=<rgba 1 1 32 8>, paramNames [WiggleTimeline, WiggleDistance, Sepreate, Size]}
  n6 = Multiply(A=LinearInterpolate(A=VertexNormalWS(), B=AppendVector(A=@n1, B=@n2), Alpha=@n3.RemoveSeperateZ), B=Multiply(A=@n4.G, B=@n5.Sepreate))
  n7 = Saturate(None=Add(A=ParticleColor().A, B=@n4.R))
  n8 = DynamicParameter() {defaultValue=<rgba 1 1 32 8>, paramNames [WiggleTimeline, WiggleDistance, Sepreate, Size]}
  n9 = Multiply(
   A=Add(A=@n8.WiggleTimeline, B=@n4.B)
   B=LinearInterpolate(
    A=ScalarParameter() {defaultValue=2, parameterName=WiggleSpeedA}
    B=ScalarParameter() {defaultValue=1, parameterName=WiggleSpeedB}
    Alpha=@n4.G
   )
  )
  n10 = Sine(None=@n9)
  n11 = Cosine(None=@n9)
  n12 = AppendVector(A=Multiply(A=AppendVector(A=@n10, B=@n11), B=@n8.WiggleDistance), B=Constant())
  n13 = Add(A=@n12, B=WorldPosition().XYZ)
  n14 = CameraPositionWS()
  n15 = ScalarParameter() {defaultValue=75, parameterName=CameraShinkRange}
  n16 = Divide(A=Subtract(A=Distance(A=@n13, B=@n14), B=@n15), B=@n15)
  n17 = MaterialFunctionCall!(Center Location, 0-1 UVs, XY scale=Multiply(A=Multiply(A=@n5.Size, B=@n7), B=Saturate(None=@n16)), OverridePivot, "Normals (Optional)") {materialFunction=&Sprite}
  n18 = Multiply(
   A=MaterialFunctionCall!(Ignore XY Offsets=StaticBoolParameter() {defaultValue=true, parameterName=Ignore XY Gerstner Offsets}, Position, WaterBodyIndex, Depth) {materialFunction=/FortWater/Materials/Functions/WaterParticleDisplacement}.WPO
   B=VectorParameter() {parameterName=Water Displacement Multiplier, defaultValue=<rgba 1 1 1 0>}.RGB
  )
  n19 = MaterialFunctionCall(Z=ScalarParameter() {defaultValue=1, parameterName=ZOffset}) {materialFunction=&MakeFloat3}
  MaterialExpressionAdd_2 = Add(
   A=Add(A=Add(A=@n6, B=@n17.World position offset), B=@n12)
   B=StaticSwitchParameter(True=Add(A=@n18, B=@n19.Result), False=Constant()) {parameterName=UseWaterDisplacement}
  )
  n20 = Custom!(Value=@MaterialExpressionAdd_2) {
   code="##if XBOXONE_PROFILE || SM5_PROFILE || SM4_PROFILE\r\n\tprecise float3 pv = Value;\r\n\treturn pv;\r\n##else\r\n\treturn Value;\r\n##endif"
   desc=Make Precise to Fix Depth Variance
   inputs [{inputName=Value, input {expression=&self:MaterialExpressionAdd_2}}]
  }
  NamedRerouteDeclaration(None=@n20) {name=WPO, nodeColor=<rgba 0 0.8594 1 1>}
  n21 = LinearInterpolate(
   B=VectorParameter() {group=WaterShader, parameterName=BiomeColorTint, defaultValue=<rgba 0.2708 0.262 0.1354 2>}.RGB
   Alpha=MaterialFunctionCall() {materialFunction=/Game/Effects/Niagara/Emitters/Water/FoamEffect/MF_NFX_BiomeMask}.Result
  ) {constA=1}
  n22 = LinearInterpolate(
   A=VectorParameter() {parameterName=SSS, defaultValue=<rgba 0.1675 0.3281 0.3281 1>}.RGB
   Alpha=TextureSample() {texture=/Game/Effects/Fort_Effects/Textures/Gameplay/T_Sphere_Mask}.R
  ) {constB=0.1}
  NamedRerouteDeclaration(None=MaterialFunctionCall!(RGB=Multiply(A=@n21, B=@n22), HSV Offset=NamedRerouteUsage(), Optional Reference Color) {materialFunction=/Game/Effects/Niagara/Emitters/Water/FoamEffect/MF_HSV_Offset}.Result) {name=SubSurface, nodeColor=<rgba 0 1 0.4647 1>}
  n23 = ParticleColor()
  n24 = LinearInterpolate(
   A=VectorParameter() {parameterName=CascadeColor_01, defaultValue=<rgba 1 0.9 0.8 0>}.RGB
   B=VectorParameter() {parameterName=CascadeColor_02, defaultValue=<rgba 0.8 0.9 1 0>}.RGB
   Alpha=@n23.R
  )
  n25 = TextureSample() {texture=/Game/Packages/Fortress_SharedMaterials/Textures/ReflectionMap2}
  n26 = Multiply(A=ScalarParameter() {defaultValue=0.5, parameterName=Emissive}, B=Multiply(A=@n25.RGB, B=@n25.RGB))
  n27 = Desaturation(None=Multiply(A=Multiply(A=@n24, B=@n23.G), B=@n26), Fraction=@n23.B)
  n28 = Add(
   A=MaterialFunctionCall!(EmissiveOvrride) {materialFunction=/Game/Effects/Niagara/Emitters/Water/FoamEffect/MF_NFX_EmissiveMask}.Result
   B=Multiply(A=StaticSwitchParameter(True=@n27, False=@n26) {parameterName=UseCascadeColorSetup?}, B=@n21)
  )
  n29 = NamedRerouteUsage()
  NamedRerouteDeclaration(None=MaterialFunctionCall(Emissive=MaterialFunctionCall(RGB=@n28, HSV Offset=@n29) {materialFunction=&MF_HSV_Offset}.Result) {materialFunction=&MF_FX_EmissiveAdjustments}.Result) {name=Emissive, nodeColor=<rgba 0.5 0 1 1>}
  n30 = ParticleColor()
  n31 = SphereMask(A=TextureCoordinate(), B=Constant() {r=0.5}, Radius=Constant() {r=0.5})
  NamedRerouteDeclaration(None=StaticSwitchParameter(True=Multiply(A=@n30.B, B=@n31), False=@n31) {parameterName=UseCascadeColorSetup?}) {name=Opacity, nodeColor=<rgba 1 0 0.6716 1>}
  n32 = VectorParameter() {parameterName=BaseColor, defaultValue=<rgba 0.0456 0.0876 0.1146 0>}
  NamedRerouteDeclaration(None=MaterialFunctionCall(RGB=Multiply(A=@n32.RGB, B=@n21), HSV Offset=NamedRerouteUsage()) {materialFunction=&MF_HSV_Offset}.Result) {name=Base, nodeColor=<rgba 0 0.2621 1 1>}
  n33 = ScalarParameter() {desc="Interpret Particle RGB as HSV Offsets, and use ParticleColor Fallback instead..", parameterName=Use Particle Color RGB as HSV Offset}
  NamedRerouteDeclaration(None=LinearInterpolate(A=Constant3Vector(), B=ParticleColor().RGB, Alpha=@n33)) {name=HSVOffset, nodeColor=<rgba 0 0.9062 1 1>}
  NamedRerouteUsage()
  NamedRerouteUsage()
  n34 = NamedRerouteUsage()
  Reroute(None=Multiply(A=ScalarParameter() {defaultValue=1, parameterName=Final Opacity Multiplier}, B=@n34))
  Reroute(None=NamedRerouteUsage())
  Reroute(None=NamedRerouteUsage())
  Reroute(None=TextureSample() {samplerType=SAMPLERTYPE_Normal, texture=&T_Sphere_N}.RGB)
  Reroute(None=@n34)
 ]
}
